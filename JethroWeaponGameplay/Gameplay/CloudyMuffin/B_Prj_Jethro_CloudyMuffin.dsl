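; CloudyMuffin knockback grenade projectile Blueprint with a two-stage explosion and water-aware effects.
; When the fuse ends in water or after a bounce, otherwise on a later bounce, it plays pre-explosion FX and applies a distance-based audio mix to nearby locally viewed pawns.
; Authority triggers the main explosion after a delay and launches non-immune hit actors outward with minimum upward velocity.
; A timer drives the pre-explosion sound tell.
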(event EventBeginPlay
  ;[0 0 0] Begin Play
  (|CallParentFunction)
  (bind _returnvalue (Utilities|Time|SetTimerbyEvent (EventDispatchers|CreateEvent :!event Custom|Pre%20Explo%20Audio%20Tell) (Variables|Audio|GetPreExplode%20Sound%20Time)))
  (Variables|Audio|SetPreExplode%20Sound%20Timer%20Handle _returnvalue))

(event Custom|Pre%20Explo%20Audio%20Tell
  (Audio|SpawnSoundAttached (Variables|Sounds|GetPreExplode%20Sound) (Variables|Transformation|GetRootComponent))
  (Audio|Components|Audio|FadeOut (Variables|B_Prj_Grenade|GetGrenadeFuse_AudioComponent) 0.1 0.0 "Linear" :!class /Script/Engine.AudioComponent))

(event Custom|ReplicateExplosion
  ;[1 1 1] Replicated Explosion
  (bind _returnvalue (Niagara|SpawnSystematLocation "/JethroWeaponGameplay/Gameplay/FX/CloudyMuffin/NS_Jethro_CloudyMuffin_Exp_Pre.NS_Jethro_CloudyMuffin_Exp_Pre" (Variables|Default|GetexplodeLocation_0_0)))
  (Niagara|SetNiagaraVariable(Bool) _returnvalue "IsInWater" (or (Water|IsinWater (Variables|Water|GetWaterInteractionComponent)) (Utilities|Enum|Equal(Enum) (Variables|Default|GetE_Surface_Type) "SurfaceType8"))))

(event EventOnExploded
  (params (HitActors :type Array%20of%20Actor%20Object%20References) (HitResults :type Array%20of%20Hit%20Result%20Structures))
  (bind (_returnvalue_location _returnvalue_rotation) (Math|Transform|BreakTransform (Transformation|GetActorTransform)))
  (Utilities|FlowControl|Delay 0.1)
  (|CallParentFunction HitActors HitResults)
  (Utilities|Time|ClearandInvalidateTimerbyHandle (Variables|Audio|GetPreExplode%20Sound%20Timer%20Handle))
  (Rendering|SetVisibility (Variables|B_Prj_Grenade|GetMesh) false false)
  ;[0 0 0 0.5] Impact VFX And Audio
  (Utilities|AllowJanus
    (:Allow
      (bind _returnvalue (Development|GetConsoleVariableIntValue "r.PhysicsField.EnableField"))
      (if (== _returnvalue 1)
        (Game|SpawnActorfromClass "/Game/Effects/Niagara/Blueprints/PhsyicsFields/FN_RadialForce.FN_RadialForce_C" (Math|Transform|MakeTransform _returnvalue_location _returnvalue_rotation) "Undefined" "MultiplyWithRoot" 0 0.2 0.66 1.5e+03 2.0 0.0 0)))
    (:Nope
      _))
  (Class|BPSurfaceTypeSoundComponent|PlaySound (Variables|Default|GetBP_SurfaceTypeSoundComponent1) (Transformation|GetActorLocation))
  (GameplayCue|ExecuteGameplayCue%28Burst%29Local self (Variables|Default|GetFeedbackCueSecond) (Ability|GameplayCue|MakeGameplayCueParameters 0.0)) ; Cam shake + controller vibration
  (bind _returnvalue_2 (Rendering|Decal|SpawnDecalatLocation "/FNGameplayCues/Athena/Prototype/Blueprints/Cube/FX/MD_Decal_Scorch_NoEmissive_CanFade.MD_Decal_Scorch_NoEmissive_CanFade" "128.000000,600.000000,600.000000" (Transformation|GetActorLocation) (Math|Rotator|MakeRotator (Math|Random|RandomFloatinRange 0.0 3.6e+02) -9e+01) 4.0))
  (Rendering|Components|Decal|SetFadeIn _returnvalue_2 0.0 0.1)
  (Rendering|Components|Decal|SetFadeOut _returnvalue_2 1.4 2.5 true)
  (Networking|SwitchHasAuthority
    (:Authority
      (Utilities|FlowControl|Sequence
        (:then_0
          ;[0 0 0 0.5] Wait for Stage 1 Explosion NS to finish then do Main Explosion. Re set Hit Location in case of transform discrepancy
          (Utilities|FlowControl|Delay 0.5)
          (Variables|Projectile|SetHitLocation (Transformation|GetActorLocation))
          (Delivery|DoExplosion)
          (Components|DestroyComponent (Variables|Default|GetNS_Object_FloatingInWater_0)))
        (:then_1
          ;[0 0 0 0.5] Knockback, unless player/pawn has immunity tag
          (for _array_element HitActors
            (if (not (GameplayTags|HasMatchingGameplayTag (Ability|GetAbilitySystemComponent _array_element) "(TagName=\"Gameplay.Status.KnockbackImmunity\")"))
              (Environment|ApplyGlobalEnvironmentGameplayEffecttoActor _array_element (Variables|Default|GetGE_KnockbackStatus_0_0) 0 "(GameplayTags=)" :!pinned true) ; Apply Knockback Status
              (Effects|SpawnFXSystemAttached :FXSystemAsset "/Game/Weapons/Prototype/FringePlank/Particle/P_FringPlank_Impact.P_FringPlank_Impact" :AttachPointName "None" :Location (Transformation|GetActorLocation _array_element) :Scale "8.000000,8.000000,8.000000" :LocationType "KeepRelativeOffset" :bAutoDestroy true :PoolingMethod "None" :bAutoActivate true)
              (bind _ascharacter (Utilities|Casting|CastToCharacter _array_element))
              (Character|LaunchCharacter _ascharacter (CallFunction|CalculateLaunchVel _ascharacter) true true)
              (bind _asfort_player_pawn_athena (Utilities|Casting|CastToFortPlayerPawnAthena _ascharacter))
              (Class|FortPlayerPawnAthena|SetFallInstigator _asfort_player_pawn_athena (Game|GetInstigatorController) (Variables|Tags|GetDefaultTags))
              (Game|SpawnActorfromClass "/Game/Effects/Niagara/Blueprints/PhsyicsFields/FN_RadialForce.FN_RadialForce_C" (Transformation|GetWorldTransform (Variables|Collision|GetCapsuleComponent)) "Undefined" "MultiplyWithRoot" 0 0.2 0.66 1.5e+03 1.0 0.0 0))))))))

(event EventFuseEnded
  (bind (_returnvalue_location _returnvalue_rotation) (Math|Transform|BreakTransform (Transformation|GetActorTransform)))
  ;[0 0 0 0.5] prevent aerial detonation by waiting for bounce UNLESS fuse ends in water, then explode
  (Variables|Default|SetfuseHasEnded true)
  (if (or (Water|IsinWater (Variables|Water|GetWaterInteractionComponent)) (Variables|Default|GethasBounced))
    (bind _output_get (Variables|Default|SetexplodeLocation_Base (Transformation|GetActorLocation self)))
    (bind _output_get_2 (Variables|Projectile|SetHitLocation _output_get))
    (bind _output_get_3 (Variables|Default|SetexplodeLocation_0_0 _output_get_2))
    ;[0 0 0 0.5] Stage 1 Explosion VFX. Only do once
    (Utilities|FlowControl|DoOnce false
      (:Completed
        (bind _returnvalue (Niagara|SpawnSystematLocation "/JethroWeaponGameplay/Gameplay/FX/CloudyMuffin/NS_Jethro_CloudyMuffin_Exp_Pre.NS_Jethro_CloudyMuffin_Exp_Pre" _output_get_3))
        (Niagara|SetNiagaraVariable(Bool) _returnvalue "IsInWater" (or (Water|IsinWater (Variables|Water|GetWaterInteractionComponent)) (Utilities|Enum|Equal(Enum) (Variables|Default|GetE_Surface_Type) "SurfaceType8")))
        (Utilities|AllowJanus
          (:Allow
            (bind _returnvalue_2 (Development|GetConsoleVariableIntValue "r.PhysicsField.EnableField"))
            (if (== _returnvalue_2 1)
              (Game|SpawnActorfromClass "/Game/Effects/Niagara/Blueprints/PhsyicsFields/FN_RadialForce.FN_RadialForce_C" (Math|Transform|MakeTransform _returnvalue_location _returnvalue_rotation) "Undefined" "MultiplyWithRoot" 0 0.2 0.66 1.2e+03 1.25 0.0 0)))
          (:Nope
            _))
        (CallFunction|ReplicateExplosion)
        (Audio|SpawnSoundAttached (Variables|Sounds|GetCue_ExplosionSound) (Variables|Transformation|GetRootComponent))
        ; Audio
        (Utilities|FlowControl|Sequence
          (:then_0
            (bind (_outhits _returnvalue_3) (Collision|MultiSphereTraceByChannel :Start (Variables|Default|GetexplodeLocation_0_0) :End (Variables|Default|GetexplodeLocation_0_0) :Radius 1.25e+03 :TraceChannel "TraceTypeQuery9" :bTraceComplex false :DrawDebugType "ForDuration" :bIgnoreSelf true :TraceColor "(R=1.000000,G=0.000000,B=0.000000,A=1.000000)" :TraceHitColor "(R=0.000000,G=1.000000,B=0.000000,A=1.000000)" :DrawTime 5.0))
            (if _returnvalue_3
              (for _array_element _outhits
                (bind _asfort_player_pawn_athena (Utilities|Casting|CastToFortPlayerPawnAthena (.HitActor (Collision|BreakHitResult _array_element))))
                (CallFunction|ApplyControlBusMix _asfort_player_pawn_athena (Math|Vector|Distance(Vector) (Transformation|GetActorLocation _asfort_player_pawn_athena) (Variables|Default|GetexplodeLocation_0_0))))))
          (:then_1
            (Rendering|SetVisibility (Variables|B_Prj_Grenade|GetMesh) false false)
            (Networking|SwitchHasAuthority
              (:Authority
                (Utilities|FlowControl|Sequence
                  (:then_0
                    ;[0 0 0 0.5] Wait for Stage 1 Explosion NS to finish then do Main Explosion. Re set Hit Location in case of transform discrepancy
                    (Utilities|FlowControl|Delay 0.5)
                    (Variables|Projectile|SetHitLocation (Transformation|GetActorLocation))
                    (Delivery|DoExplosion)
                    (Components|DestroyComponent (Variables|Default|GetNS_Object_FloatingInWater_0)))
                  (:then_1
                    ;[0 0 0 0.5] Knockback, unless player/pawn has immunity tag
                    (for _array_element_2 (stale EventOnExploded HitActors)
                      (if (not (GameplayTags|HasMatchingGameplayTag (Ability|GetAbilitySystemComponent _array_element_2) "(TagName=\"Gameplay.Status.KnockbackImmunity\")"))
                        (Environment|ApplyGlobalEnvironmentGameplayEffecttoActor _array_element_2 (Variables|Default|GetGE_KnockbackStatus_0_0) 0 "(GameplayTags=)" :!pinned true) ; Apply Knockback Status
                        (Effects|SpawnFXSystemAttached :FXSystemAsset "/Game/Weapons/Prototype/FringePlank/Particle/P_FringPlank_Impact.P_FringPlank_Impact" :AttachPointName "None" :Location (Transformation|GetActorLocation _array_element_2) :Scale "8.000000,8.000000,8.000000" :LocationType "KeepRelativeOffset" :bAutoDestroy true :PoolingMethod "None" :bAutoActivate true)
                        (bind _ascharacter (Utilities|Casting|CastToCharacter _array_element_2))
                        (Character|LaunchCharacter _ascharacter (CallFunction|CalculateLaunchVel _ascharacter) true true)
                        (bind _asfort_player_pawn_athena_2 (Utilities|Casting|CastToFortPlayerPawnAthena _ascharacter))
                        (Class|FortPlayerPawnAthena|SetFallInstigator _asfort_player_pawn_athena_2 (Game|GetInstigatorController) (Variables|Tags|GetDefaultTags))
                        (Game|SpawnActorfromClass "/Game/Effects/Niagara/Blueprints/PhsyicsFields/FN_RadialForce.FN_RadialForce_C" (Transformation|GetWorldTransform (Variables|Collision|GetCapsuleComponent)) "Undefined" "MultiplyWithRoot" 0 0.2 0.66 1.5e+03 1.0 0.0 0)))))))))))))

(event Custom|Apply%20Control%20Bus%20Mix
  (params (Hit%20Pawn :type Fort%20Player%20Pawn%20Athena%20Object%20Reference) (Distance%20from%20Effect :type Float%20%28double-precision%29))
  ; Audio
  (Utilities|IsValid Hit%20Pawn
    (:Is%20Valid
      (if (Pawn|IsLocallyViewed Hit%20Pawn)
        (Utilities|FlowControl|DoOnce false
          (:Completed
            (Game|AddComponentbyClass 0 "/FortMetaSoundWeapon/Items/Modulation/DynamicControlBusMixComponent/BP_DynamicControlBusMixComponent.BP_DynamicControlBusMixComponent_C" Distance%20from%20Effect (Variables|Default|GetDynamic%20Control%20Bus%20Mix%20Data))))))))

(event EventOnBounce
  (params (Hit :type Hit%20Result%20Structure%20%28by%20ref%29))
  (bind (_bblockinghit _binitialoverlap _time _distance _location _impactpoint _normal _impactnormal _physmat _hitactor) (Collision|BreakHitResult Hit))
  (bind (_returnvalue_location _returnvalue_rotation) (Math|Transform|BreakTransform (Transformation|GetActorTransform)))
  (bind _surfacetype (Variables|/Script/PhysicsCore.PhysicalMaterial|GetSurfaceType _physmat))
  ;[0 0 0] Bounce
  (|CallParentFunction Hit)
  (Utilities|IsValid _physmat
    (:Is%20Valid
      (Variables|Default|SethasBounced true)
      (Class|BPSurfaceTypeSoundComponent|SetSurfaceType (Variables|Default|GetBP_SurfaceTypeSoundComponent1) _surfacetype _hitactor)
      ;;[1 0.5052 0.5052] To Do. It's mud, but would better to have other surface type bounce FX
      (if (== _surfacetype 30)
        (Niagara|SpawnSystematLocation (Variables|Default|GetNiagara_Bounse_0) _location)
        (else
          (Niagara|SpawnSystematLocation "/Game/Effects/Fort_Effects/Effects/Weapon/NS_impact_smoke_Small.NS_impact_smoke_Small" _location)
          (if (Variables|Default|GetfuseHasEnded)
            (bind _output_get (Variables|Default|SetexplodeLocation_Base (Transformation|GetActorLocation self)))
            (bind _output_get_2 (Variables|Projectile|SetHitLocation _output_get))
            (bind _output_get_3 (Variables|Default|SetexplodeLocation_0_0 _output_get_2))
            ;[0 0 0 0.5] Stage 1 Explosion VFX. Only do once
            (Utilities|FlowControl|DoOnce false
              (:Completed
                (bind _returnvalue (Niagara|SpawnSystematLocation "/JethroWeaponGameplay/Gameplay/FX/CloudyMuffin/NS_Jethro_CloudyMuffin_Exp_Pre.NS_Jethro_CloudyMuffin_Exp_Pre" _output_get_3))
                (Niagara|SetNiagaraVariable(Bool) _returnvalue "IsInWater" (or (Water|IsinWater (Variables|Water|GetWaterInteractionComponent)) (Utilities|Enum|Equal(Enum) (Variables|Default|GetE_Surface_Type) "SurfaceType8")))
                (Utilities|AllowJanus
                  (:Allow
                    (bind _returnvalue_2 (Development|GetConsoleVariableIntValue "r.PhysicsField.EnableField"))
                    (if (== _returnvalue_2 1)
                      (Game|SpawnActorfromClass "/Game/Effects/Niagara/Blueprints/PhsyicsFields/FN_RadialForce.FN_RadialForce_C" (Math|Transform|MakeTransform _returnvalue_location _returnvalue_rotation) "Undefined" "MultiplyWithRoot" 0 0.2 0.66 1.2e+03 1.25 0.0 0)))
                  (:Nope
                    _))
                (CallFunction|ReplicateExplosion)
                (Audio|SpawnSoundAttached (Variables|Sounds|GetCue_ExplosionSound) (Variables|Transformation|GetRootComponent))
                ; Audio
                (Utilities|FlowControl|Sequence
                  (:then_0
                    (bind (_outhits _returnvalue_3) (Collision|MultiSphereTraceByChannel :Start (Variables|Default|GetexplodeLocation_0_0) :End (Variables|Default|GetexplodeLocation_0_0) :Radius 1.25e+03 :TraceChannel "TraceTypeQuery9" :bTraceComplex false :DrawDebugType "ForDuration" :bIgnoreSelf true :TraceColor "(R=1.000000,G=0.000000,B=0.000000,A=1.000000)" :TraceHitColor "(R=0.000000,G=1.000000,B=0.000000,A=1.000000)" :DrawTime 5.0))
                    (if _returnvalue_3
                      (for _array_element _outhits
                        (bind _asfort_player_pawn_athena (Utilities|Casting|CastToFortPlayerPawnAthena (.HitActor (Collision|BreakHitResult _array_element))))
                        (CallFunction|ApplyControlBusMix _asfort_player_pawn_athena (Math|Vector|Distance(Vector) (Transformation|GetActorLocation _asfort_player_pawn_athena) (Variables|Default|GetexplodeLocation_0_0))))))
                  (:then_1
                    (Rendering|SetVisibility (Variables|B_Prj_Grenade|GetMesh) false false)
                    (Networking|SwitchHasAuthority
                      (:Authority
                        (Utilities|FlowControl|Sequence
                          (:then_0
                            ;[0 0 0 0.5] Wait for Stage 1 Explosion NS to finish then do Main Explosion. Re set Hit Location in case of transform discrepancy
                            (Utilities|FlowControl|Delay 0.5)
                            (Variables|Projectile|SetHitLocation (Transformation|GetActorLocation))
                            (Delivery|DoExplosion)
                            (Components|DestroyComponent (Variables|Default|GetNS_Object_FloatingInWater_0)))
                          (:then_1
                            ;[0 0 0 0.5] Knockback, unless player/pawn has immunity tag
                            (for _array_element_2 (stale EventOnExploded HitActors)
                              (if (not (GameplayTags|HasMatchingGameplayTag (Ability|GetAbilitySystemComponent _array_element_2) "(TagName=\"Gameplay.Status.KnockbackImmunity\")"))
                                (Environment|ApplyGlobalEnvironmentGameplayEffecttoActor _array_element_2 (Variables|Default|GetGE_KnockbackStatus_0_0) 0 "(GameplayTags=)" :!pinned true) ; Apply Knockback Status
                                (Effects|SpawnFXSystemAttached :FXSystemAsset "/Game/Weapons/Prototype/FringePlank/Particle/P_FringPlank_Impact.P_FringPlank_Impact" :AttachPointName "None" :Location (Transformation|GetActorLocation _array_element_2) :Scale "8.000000,8.000000,8.000000" :LocationType "KeepRelativeOffset" :bAutoDestroy true :PoolingMethod "None" :bAutoActivate true)
                                (bind _ascharacter (Utilities|Casting|CastToCharacter _array_element_2))
                                (Character|LaunchCharacter _ascharacter (CallFunction|CalculateLaunchVel _ascharacter) true true)
                                (bind _asfort_player_pawn_athena_2 (Utilities|Casting|CastToFortPlayerPawnAthena _ascharacter))
                                (Class|FortPlayerPawnAthena|SetFallInstigator _asfort_player_pawn_athena_2 (Game|GetInstigatorController) (Variables|Tags|GetDefaultTags))
                                (Game|SpawnActorfromClass "/Game/Effects/Niagara/Blueprints/PhsyicsFields/FN_RadialForce.FN_RadialForce_C" (Transformation|GetWorldTransform (Variables|Collision|GetCapsuleComponent)) "Undefined" "MultiplyWithRoot" 0 0.2 0.66 1.5e+03 1.0 0.0 0)))))))))))))))))

(event OnEnterWaterEvent(WaterInteractionComponent)
  (params (WaterBody :type Fort%20Water%20Body%20Actor%20Object%20Reference) (WaterInteractionComponent :type Fort%20Water%20Interaction%20Component%20Object%20Reference) (bIsFirstBody :type Boolean))
  (bind _ns_object_floatinginwater_0 (Variables|Default|GetNS_Object_FloatingInWater_0))
  ;[1 1 1] enter water and Get water info
  (bind _outwatersurfaceinfo (WaterBody|GetWaterSurfaceInfoStruct WaterBody (Transformation|GetActorLocation) true false false))
  (Components|Activation|Activate _ns_object_floatinginwater_0 false)
  (Effects|Components|ParticleSystem|SetFloatParameter _ns_object_floatinginwater_0 "User.WaterDepth" (.WaterDepth (Utilities|Struct|BreakWaterSurfaceInfo _outwatersurfaceinfo)))
  (Effects|Components|ParticleSystem|SetVectorParameter _ns_object_floatinginwater_0 "User.WaterPlaneLocation" (Utilities|Struct|BreakWaterSurfaceInfo _outwatersurfaceinfo))
  (Effects|Components|ParticleSystem|SetVectorParameter _ns_object_floatinginwater_0 "User.WaterPlaneNormal" (.WaterPlaneNormal (Utilities|Struct|BreakWaterSurfaceInfo _outwatersurfaceinfo)))
  (Effects|Components|ParticleSystem|SetFloatParameter _ns_object_floatinginwater_0 "User.InheritVelocityScale" 0.3))

(event OnExitWaterEvent(WaterInteractionComponent)
  (params (WaterBody :type Fort%20Water%20Body%20Actor%20Object%20Reference) (WaterInteractionComponent :type Fort%20Water%20Interaction%20Component%20Object%20Reference) (bIsLastBody :type Boolean))
  ;[0 0 0] Exit water
  (Components|Activation|Deactivate (Variables|Default|GetNS_Object_FloatingInWater_0)))

(fn CalculateLaunchVel
  (params (Target :type Actor%20Object%20Reference))
  (return (Math|Vector|MakeVector (.x (* (Math|Vector|Normalize (- (Transformation|GetActorLocation Target) (CallFunction|GetLaunchOriginPoint))) (ScalableFloat|GetValueatLevel (Variables|Default|GetLaunch%20Velocity)))) (.y (* (Math|Vector|Normalize (- (Transformation|GetActorLocation Target) (CallFunction|GetLaunchOriginPoint))) (ScalableFloat|GetValueatLevel (Variables|Default|GetLaunch%20Velocity)))) (Math|Float|Max(Float) (+ (.z (* (Math|Vector|Normalize (- (Transformation|GetActorLocation Target) (CallFunction|GetLaunchOriginPoint))) (ScalableFloat|GetValueatLevel (Variables|Default|GetLaunch%20Velocity)))) 5e+01) 575.0))))

(fn GetExplosion
  (params (InWater :type Boolean))
  (return (|CallParentFunction false)))

(fn GetLaunchOriginPoint
  (return (Variables|Projectile|GetHitLocation)))

(fn ConstructionScript)
