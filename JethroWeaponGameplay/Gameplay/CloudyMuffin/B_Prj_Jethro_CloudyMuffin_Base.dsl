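; Base projectile Blueprint for the Jethro weapon gameplay "CloudyMuffin" grenade, configuring its cosmetic components.
; Its construction script runs only where cosmetic events can execute, so it is skipped elsewhere.
; It points a rotating movement component at the inherited grenade mesh and assigns the bounce cue as the first collision sound on the Fort collision audio component.
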(fn ConstructionScript
  (Networking|CanExecuteCosmeticEvents
    (:True
      (Components|Movement|SetUpdatedComponent (Variables|Default|GetRotatingMovement_0) (Variables|B_Prj_Grenade|GetMesh_0))
      (Audio|SetCollisionSoundatIndex (Variables|Default|GetFortCollisionAudio) 0 (Variables|Sounds|GetCue_Bounce)))))
